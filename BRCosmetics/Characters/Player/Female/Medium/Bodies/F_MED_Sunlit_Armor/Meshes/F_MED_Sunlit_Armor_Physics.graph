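# Physics asset for the female medium Sunlit Armor body, setting up a dangling front flap for physics-driven motion.
# Seven small capsules form a chain hanging from spine_01, each joint having locked twist and narrow swing limits.
# Larger capsules on pelvis, spine_01 and both thighs give body shapes around the flap, with spine and thighs joined to the pelvis by angularly free constraints.
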PhysicsAsset /BRCosmetics/Characters/Player/Female/Medium/Bodies/F_MED_Sunlit_Armor/Meshes/F_MED_Sunlit_Armor_Physics
bodies {
 dyn_frontflap_1 [{shapeType=Capsule, length=5, radius=2.2, shapeName=dyn_frontflap_5_capsule, center=<3 0 0>, rotation=<rot 90 0 0>}]
 dyn_frontflap_2 [{shapeType=Capsule, length=5, radius=2.3, shapeName=dyn_frontflap_5_capsule, center=<3 0 0>, rotation=<rot 90 0 0>}]
 dyn_frontflap_3 [{shapeType=Capsule, length=5, radius=2.5, shapeName=dyn_frontflap_5_capsule, center=<3 0 0>, rotation=<rot 90 0 0>}]
 dyn_frontflap_4 [{shapeType=Capsule, length=5, radius=2, shapeName=dyn_frontflap_5_capsule, center=<3 0 0>, rotation=<rot 90 0 0>}]
 dyn_frontflap_5 [{shapeType=Capsule, length=5, radius=2, shapeName=dyn_frontflap_5_capsule, center=<3 0 0>, rotation=<rot 90 0 0>}]
 dyn_frontflap_6 [{shapeType=Capsule, length=10, radius=2, shapeName=dyn_frontflap_6_capsule, center=<3 0 0>, rotation=<rot 90 0 0>}]
 dyn_frontflap_7 [{shapeType=Capsule, length=10, radius=2, shapeName=dyn_frontflap_6_capsule, center=<3 0 0>, rotation=<rot 90 0 0>}]
 pelvis [{
   shapeType=Capsule
   length=13.749
   radius=9.9541
   shapeName=pelvis_capsule
   center=<-2.7268 -2.6994 0.0349>
   rotation=<rot -1.6206 -90 0.1077>
  }]
 spine_01 [{shapeType=Capsule, length=15, radius=8.595, shapeName=spine_01_capsule, center=<0 -1.7269 0>}]
 thigh_l [{shapeType=Capsule, length=33.7509, radius=7.5, shapeName=thigh_l_capsule, center=<-17.5739 0.2459 0>, rotation=<rot 90 179.7016 -178.4271>}]
 thigh_r [{shapeType=Capsule, length=33.7509, radius=7.5, shapeName=thigh_l_capsule, center=<17.5739 -0.2459 0>, rotation=<rot 90 179.7016 -178.4271>}]
}
constraints [
 {bone1Name=dyn_frontflap_1, bone2Name=spine_01, swing1LimitDegrees=8, swing2LimitDegrees=10, twistLimitDegrees=45, twistMotion=Locked}
 {bone1Name=dyn_frontflap_2, bone2Name=dyn_frontflap_1, swing1LimitDegrees=8, swing2LimitDegrees=10, twistLimitDegrees=45, twistMotion=Locked}
 {bone1Name=dyn_frontflap_3, bone2Name=dyn_frontflap_2, swing1LimitDegrees=5, swing2LimitDegrees=10, twistLimitDegrees=45, twistMotion=Locked}
 {bone1Name=dyn_frontflap_4, bone2Name=dyn_frontflap_3, swing1LimitDegrees=10, swing2LimitDegrees=10, twistLimitDegrees=45, twistMotion=Locked}
 {bone1Name=dyn_frontflap_5, bone2Name=dyn_frontflap_4, swing1LimitDegrees=10, swing2LimitDegrees=10, twistLimitDegrees=45, twistMotion=Locked}
 {bone1Name=dyn_frontflap_6, bone2Name=dyn_frontflap_5, swing1LimitDegrees=10, swing2LimitDegrees=10, twistLimitDegrees=45, twistMotion=Locked}
 {bone1Name=dyn_frontflap_7, bone2Name=dyn_frontflap_6, swing1LimitDegrees=10, swing2LimitDegrees=10, twistLimitDegrees=45, twistMotion=Locked}
 {
  bone1Name=spine_01
  bone2Name=pelvis
  swing1LimitDegrees=45
  swing1Motion=Free
  swing2LimitDegrees=45
  swing2Motion=Free
  twistLimitDegrees=45
  twistMotion=Free
 }
 {
  bone1Name=thigh_l
  bone2Name=pelvis
  swing1LimitDegrees=45
  swing1Motion=Free
  swing2LimitDegrees=45
  swing2Motion=Free
  twistLimitDegrees=45
  twistMotion=Free
 }
 {
  bone1Name=thigh_r
  bone2Name=pelvis
  swing1LimitDegrees=45
  swing1Motion=Free
  swing2LimitDegrees=45
  swing2Motion=Free
  twistLimitDegrees=45
  twistMotion=Free
 }
]
