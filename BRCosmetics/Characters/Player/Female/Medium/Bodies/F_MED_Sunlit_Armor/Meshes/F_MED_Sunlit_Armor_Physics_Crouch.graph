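# Physics asset for the female medium Sunlit Armor body, named as its crouch variant, covering the lower body and dangling front flap.
# Capsules on the pelvis, thighs and calves give the legs simple collision shapes, with a tiny spine_01 capsule as the flap's anchor.
# Six small capsules form the flap chain from spine_01, its links constrained to narrow swing and zero twist.
# The flap root and body joints swing and twist freely.
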PhysicsAsset /BRCosmetics/Characters/Player/Female/Medium/Bodies/F_MED_Sunlit_Armor/Meshes/F_MED_Sunlit_Armor_Physics_Crouch
bodies {
 calf_l [{shapeType=Capsule, length=30, radius=7, shapeName=calf_l_capsule, center=<-16 -2 0>, rotation=<rot 90 0 -8>}]
 calf_r [{shapeType=Capsule, length=30, radius=7, shapeName=calf_l_capsule, center=<16 2 0>, rotation=<rot 90 0 -8>}]
 dyn_frontflap_2 [{shapeType=Capsule, length=5, radius=1, shapeName=dyn_frontflap_5_capsule, center=<4 0 0>, rotation=<rot 90 0 0>}]
 dyn_frontflap_3 [{shapeType=Capsule, length=5, radius=1.4, shapeName=dyn_frontflap_5_capsule, center=<4 0 0>, rotation=<rot 90 0 0>}]
 dyn_frontflap_4 [{shapeType=Capsule, length=5, radius=1.4, shapeName=dyn_frontflap_5_capsule, center=<4 0 0>, rotation=<rot 90 0 0>}]
 dyn_frontflap_5 [{shapeType=Capsule, length=5, radius=1.5, shapeName=dyn_frontflap_5_capsule, center=<4 0 0>, rotation=<rot 90 0 0>}]
 dyn_frontflap_6 [{shapeType=Capsule, length=5, radius=1.5, shapeName=dyn_frontflap_6_capsule, center=<4 0 0>, rotation=<rot 90 0 0>}]
 dyn_frontflap_7 [{shapeType=Capsule, length=15, radius=2, shapeName=dyn_frontflap_6_capsule, center=<7.4739 0 0>, rotation=<rot 90 0 0>}]
 pelvis [{shapeType=Capsule, length=14, radius=11, shapeName=pelvis_capsule, center=<-4.5624 -3.3153 -0.0836>}]
 spine_01 [{shapeType=Capsule, length=1, radius=1, shapeName=spine_01_capsule}]
 thigh_l [{shapeType=Capsule, length=43, radius=8, shapeName=thigh_l_capsule, center=<-22.7149 0 -1.3388>, rotation=<rot 88.4477 179.9999 179.9999>}]
 thigh_r [{shapeType=Capsule, length=43, radius=9, shapeName=thigh_l_capsule, center=<22.7512 -1.2487 -1.2388>, rotation=<rot 90 0 0>}]
}
constraints [
 {bone1Name=dyn_frontflap_3, bone2Name=dyn_frontflap_2, swing1LimitDegrees=20, swing2LimitDegrees=3, twistLimitDegrees=0}
 {bone1Name=dyn_frontflap_4, bone2Name=dyn_frontflap_3, swing1LimitDegrees=20, swing2LimitDegrees=3, twistLimitDegrees=0}
 {bone1Name=dyn_frontflap_5, bone2Name=dyn_frontflap_4, swing1LimitDegrees=20, swing2LimitDegrees=3, twistLimitDegrees=0}
 {bone1Name=dyn_frontflap_6, bone2Name=dyn_frontflap_5, swing1LimitDegrees=20, swing2LimitDegrees=3, twistLimitDegrees=0}
 {bone1Name=dyn_frontflap_7, bone2Name=dyn_frontflap_6, swing1LimitDegrees=20, swing2LimitDegrees=3, twistLimitDegrees=0}
 {
  bone1Name=calf_l
  bone2Name=thigh_l
  swing1LimitDegrees=45
  swing1Motion=Free
  swing2LimitDegrees=45
  swing2Motion=Free
  twistLimitDegrees=45
  twistMotion=Free
 }
 {
  bone1Name=dyn_frontflap_2
  bone2Name=spine_01
  swing1LimitDegrees=45
  swing1Motion=Free
  swing2LimitDegrees=45
  swing2Motion=Free
  twistLimitDegrees=45
  twistMotion=Free
 }
 {
  bone1Name=spine_01
  bone2Name=pelvis
  swing1LimitDegrees=45
  swing1Motion=Free
  swing2LimitDegrees=45
  swing2Motion=Free
  twistLimitDegrees=45
  twistMotion=Free
 }
 {
  bone1Name=thigh_l
  bone2Name=pelvis
  swing1LimitDegrees=45
  swing1Motion=Free
  swing2LimitDegrees=45
  swing2Motion=Free
  twistLimitDegrees=45
  twistMotion=Free
 }
 {
  bone1Name=thigh_r
  bone2Name=pelvis
  swing1LimitDegrees=45
  swing1Motion=Free
  swing2LimitDegrees=45
  swing2Motion=Free
  twistLimitDegrees=45
  twistMotion=Free
 }
 {
  bone1Name=calf_r
  bone2Name=thigh_r
  swing1LimitDegrees=45
  swing1Motion=Free
  swing2LimitDegrees=45
  swing2Motion=Free
  twistLimitDegrees=45
  twistMotion=Free
 }
]
