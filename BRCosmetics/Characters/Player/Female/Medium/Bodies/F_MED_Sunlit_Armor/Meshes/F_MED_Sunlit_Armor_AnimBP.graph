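# Animation Blueprint for the Sunlit Armor female body cosmetic, layering secondary motion and corrective bone offsets onto the parent mesh's pose.
# The copied pose passes through a Control Rig, attach constraints and hip, neck and shoulder correctives driven by limb angles measured in EventGraph.
# That cached pose feeds separate standing and crouch RigidBody simulations, blended by crouch state.
# Front-flap bone offsets handle crouch, jump and a custom frontend idle montage, including an LOD1 fix.
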AnimBlueprint /BRCosmetics/Characters/Player/Female/Medium/Bodies/F_MED_Sunlit_Armor/Meshes/F_MED_Sunlit_Armor_AnimBP
target_skeleton=&/F_MED_Sunlit_Armor_Skeleton
graphs [
 AnimGraph {nodes [
   n1 = "Constraint - dyn_frontflap_1 (R)"(ComponentPose=Local To Component(LocalPose=Use cached pose 'startPose'()), Alpha=1, ConstraintWeights_0=1) {
    alphaBoolBlend {blendOption=Linear}
    boneToModify {boneName=dyn_frontflap_1}
    constraintSetup [{transformType=Rotation, targetBone {boneName=thigh_l}}]
    constraintWeights [1]
   }
   n2 = "Constraint - dyn_frontflap_2 (R)"(ComponentPose=@n1, Alpha=1, ConstraintWeights_0=1) {
    alphaBoolBlend {blendOption=Linear}
    boneToModify {boneName=dyn_frontflap_2}
    constraintSetup [{transformType=Rotation, targetBone {boneName=thigh_l}}]
    constraintWeights [1]
   }
   n3 = "Transform (Modify) Bone - Bone: dyn_frontflap_1"!(ComponentPose=@n2, Alpha=Get isAccelerating2D() -> (isAccelerating2D:double)) -> (Pose) {
    rotationMode=BMM_Additive
    rotationSpace=BCS_BoneSpace
    translationMode=BMM_Additive
    alphaBoolBlend {blendOption=Linear}
    alphaScaleBiasClamp {bInterpResult=true}
    boneToModify {boneName=dyn_frontflap_1}
    rotation {roll=45}
   }
   n4 = Get isNotAccelerating2D() -> (isNotAccelerating2D:double)
   n5 = "Transform (Modify) Bone - Bone: dyn_frontflap_1"(ComponentPose=@n3, Alpha=@n4) {
    rotationMode=BMM_Additive
    rotationSpace=BCS_BoneSpace
    translationMode=BMM_Additive
    alphaBoolBlend {blendOption=Linear}
    alphaScaleBiasClamp {bInterpResult=true}
    boneToModify {boneName=dyn_frontflap_1}
    rotation {roll=72}
    translation=<0.7 1.8 -4.7>
   }
   n6 = "Transform (Modify) Bone - Bone: dyn_frontflap_2"!(ComponentPose=@n5, Alpha=@n4) -> (Pose) {
    rotationMode=BMM_Additive
    rotationSpace=BCS_BoneSpace
    alphaBoolBlend {blendOption=Linear}
    alphaScaleBiasClamp {bInterpResult=true}
    boneToModify {boneName=dyn_frontflap_2}
    rotation {pitch=6}
   }
   n7 = RigidBody(ComponentPose=@n6) {
    bClampLinearTranslationLimitToRefPose=true
    bOverrideWorldGravity=true
    overridePhysicsAsset=&/F_MED_Sunlit_Armor_Physics_Crouch
    alphaBoolBlend {blendOption=Linear}
    baseBoneRef {boneName=pelvis}
    componentAppliedLinearAccClamp=<3000 3000 3000>
    componentLinearAccScale=<0.01 0.01 0.01>
    componentLinearVelScale=<0.01 0.01 0.01>
    overrideWorldGravity {z=-6000}
   }
   n8 = RigidBody!(
    ComponentPose=Local To Component(LocalPose=Use cached pose 'startPose'())
    Alpha
    OverrideWorldGravity:Vector=Get Gravity Override()
    ExternalForce
   ) -> (Pose) {
    bClampLinearTranslationLimitToRefPose=true
    bOverrideWorldGravity=true
    overridePhysicsAsset=&/F_MED_Sunlit_Armor_Physics
    alphaBoolBlend {blendOption=Linear}
    baseBoneRef {boneName=pelvis}
    componentAppliedLinearAccClamp=<3000 3000 3000>
    componentLinearAccScale=<0.1 0.1 0.1>
    componentLinearVelScale=<0.3 0.3 0.31>
    overrideWorldGravity {z=-2000}
    simSpaceSettings {velocityScaleZ=0.5}
   }
   n9 = Get jumpFalling() -> (jumpFalling:double)
   n10 = "Transform (Modify) Bone - Bone: dyn_frontflap_1"!(ComponentPose=@n8, Alpha=@n9, Translation:Vector="0,0,0", Scale:Vector="1,1,1") -> (Pose) {
    rotationMode=BMM_Additive
    rotationSpace=BCS_BoneSpace
    alphaBoolBlend {blendOption=Linear}
    alphaScaleBiasClamp {bInterpResult=true, interpSpeedDecreasing=5, interpSpeedIncreasing=5}
    boneToModify {boneName=dyn_frontflap_1}
    rotation {pitch=30, yaw=-10}
   }
   n11 = "Transform (Modify) Bone - Bone: dyn_frontflap_2"!(ComponentPose=@n10, Alpha=@n9, Translation:Vector="0,0,0", Scale:Vector="1,1,1") -> (Pose) {
    rotationMode=BMM_Additive
    rotationSpace=BCS_BoneSpace
    alphaBoolBlend {blendOption=Linear}
    alphaScaleBiasClamp {bInterpResult=true, interpSpeedDecreasing=5, interpSpeedIncreasing=5}
    boneToModify {boneName=dyn_frontflap_2}
    rotation {pitch=20}
   }
   n12 = "Transform (Modify) Bone - Bone: dyn_frontflap_3"!(ComponentPose=@n11, Alpha=@n9, Translation="0,0,0", Scale="1,1,1") -> (Pose) {
    rotationMode=BMM_Additive
    rotationSpace=BCS_BoneSpace
    alphaBoolBlend {blendOption=Linear}
    alphaScaleBiasClamp {bInterpResult=true, interpSpeedDecreasing=5, interpSpeedIncreasing=5}
    boneToModify {boneName=dyn_frontflap_3}
    rotation {pitch=20}
   }
   n13 = "Transform (Modify) Bone - Bone: dyn_frontflap_4"!(ComponentPose=@n12, Alpha=@n9, Translation="0,0,0", Scale="1,1,1") -> (Pose) {
    rotationMode=BMM_Additive
    rotationSpace=BCS_BoneSpace
    alphaBoolBlend {blendOption=Linear}
    alphaScaleBiasClamp {bInterpResult=true, interpSpeedDecreasing=5, interpSpeedIncreasing=5}
    boneToModify {boneName=dyn_frontflap_4}
    rotation {pitch=20}
   }
   n14 = Get customFrontendIdleOn() -> (customFrontendIdleOn:bool)
   n15 = "Transform (Modify) Bone - Bone: dyn_frontflap_1"!(ComponentPose=@n13, bAlphaBoolEnabled:bool=@n14, Translation="0,0,0", Rotation:Rotator="-12,2,-4.7", Scale="1,1,1") -> (Pose) {
    alphaInputType=Bool
    rotationMode=BMM_Additive
    rotationSpace=BCS_BoneSpace
    alphaBoolBlend {blendOption=Linear}
    boneToModify {boneName=dyn_frontflap_1}
    rotation=<rot -12 2 -4.7>
   }
   n16 = "Transform (Modify) Bone - Bone: dyn_frontflap_2"!(ComponentPose=@n15, bAlphaBoolEnabled:bool=@n14, Translation="0,0,0", Rotation:Rotator="2.337,10.131,5.781", Scale="1,1,1") -> (Pose) {
    alphaInputType=Bool
    rotationMode=BMM_Additive
    rotationSpace=BCS_BoneSpace
    alphaBoolBlend {blendOption=Linear}
    boneToModify {boneName=dyn_frontflap_2}
    rotation=<rot 2.337 10.131 5.781>
   }
   n17 = "Transform (Modify) Bone - Bone: dyn_frontflap_3"!(ComponentPose=@n16, bAlphaBoolEnabled:bool=@n14, Translation="0,0,-0.8", Rotation:Rotator="0,0,0", Scale="1,1,1") -> (Pose) {
    alphaInputType=Bool
    translationMode=BMM_Additive
    translationSpace=BCS_BoneSpace
    alphaBoolBlend {blendOption=Linear}
    boneToModify {boneName=dyn_frontflap_3}
    translation {z=-0.8}
   }
   n18 = "Transform (Modify) Bone - Bone: dyn_frontflap_1"(
    ComponentPose=@n17
    bAlphaBoolEnabled=Get customFrontendIdleOnAtLOD1() -> (customFrontendIdleOnAtLOD1:bool)
    Translation="0,0,0.3"
    Rotation="0,0,0"
    Scale="1,1,1"
   ) {
    alphaInputType=Bool
    translationMode=BMM_Additive
    translationSpace=BCS_BoneSpace
    alphaBoolBlend {blendOption=Linear}
    boneToModify {boneName=dyn_frontflap_1}
    translation {z=0.3}
   }
   Output Pose(Result=Blend Poses by bool(BlendPose_0=Component To Local(ComponentPose=@n7), BlendPose_1=Component To Local(ComponentPose=@n18), bActiveValue=Get bIsCrouching()) {blendType=Linear}) {layerGroup=None}
   n19 = Get R_thighForward() -> (R_thighForward:double)
   n20 = Control Rig(Source=Copy Pose From Mesh() {bCopyCurves=true, bUseAttachedParent=true}) {
    alphaBoolBlend {blendOption=Linear}
    controlRigAssetReference {blueprintRigClass=&/F_MED_Sunlit_Armor_CtrlRig.F_MED_Sunlit_Armor_CtrlRig_C}
   }
   n21 = "Constraint - attach_backpack (PPP)"!(
    ComponentPose=Local To Component(LocalPose=@n20)
    Alpha=1
    ConstraintWeights_0=0.8431
    ConstraintWeights_1:float=0.1515
    ConstraintWeights_2:float=0.0054
   ) -> (Pose) {
    alphaBoolBlend {blendOption=Linear}
    alphaScaleBiasClamp {bInterpResult=true, interpSpeedDecreasing=15, interpSpeedIncreasing=15}
    boneToModify {boneName=attach_backpack}
    constraintSetup [{targetBone {boneName=spine_05}}, {targetBone {boneName=spine_04}}, {targetBone {boneName=clavicle_r}}]
    constraintWeights [1,1,1]
   }
   n22 = "Constraint - attach_cape (P)"!(ComponentPose=@n21, Alpha=1, ConstraintWeights_0=1, ConstraintWeights_1:float=0) -> (Pose) {
    alphaBoolBlend {blendOption=Linear}
    alphaScaleBiasClamp {bInterpResult=true, interpSpeedDecreasing=15, interpSpeedIncreasing=15}
    boneToModify {boneName=attach_cape}
    constraintSetup [{targetBone {boneName=spine_05}}]
    constraintWeights [1,0]
   }
   n23 = "Transform (Modify) Bone - Bone: dyn_L_neckCrest"!(ComponentPose=@n22, Alpha=Get L upperarm up() -> (L upperarm up:double), Translation="0,0,0") -> (Pose) {
    rotationMode=BMM_Additive
    scaleMode=BMM_Additive
    alphaBoolBlend {blendOption=Linear}
    boneToModify {boneName=dyn_L_neckCrest}
    scale {x=0.7}
   }
   n24 = Get neckTiltedBack() -> (neckTiltedBack:double)
   n25 = "Transform (Modify) Bone - Bone: dyn_back_neckCrest"!(ComponentPose=@n23, Alpha=@n24, Translation="0,0,-3.5", Scale="1,1,1") -> (Pose) {
    translationMode=BMM_Additive
    translationSpace=BCS_BoneSpace
    alphaBoolBlend {blendOption=Linear}
    alphaScaleBiasClamp {bInterpResult=true}
    boneToModify {boneName=dyn_back_neckCrest}
    translation {z=-3.5}
   }
   n26 = "Transform (Modify) Bone - Bone: dyn_R_neckCrest"!(ComponentPose=@n25, Alpha=@n24, Translation="0,0,3", Scale="1,1,1") -> (Pose) {
    translationMode=BMM_Additive
    translationSpace=BCS_BoneSpace
    alphaBoolBlend {blendOption=Linear}
    alphaScaleBiasClamp {bInterpResult=true}
    boneToModify {boneName=dyn_R_neckCrest}
    translation {z=3}
   }
   n27 = "Transform (Modify) Bone - Bone: dyn_R_shoulderSpike"!(ComponentPose=@n26, Alpha=Get neckTilted R() -> (neckTilted R:double), Scale="0.7,1,1") -> (Pose) {
    scaleMode=BMM_Additive
    scaleSpace=BCS_BoneSpace
    alphaBoolBlend {blendOption=Linear}
    boneToModify {boneName=dyn_R_shoulderSpike}
    scale {x=0.7}
   }
   n28 = "Transform (Modify) Bone - Bone: dyn_L_hipFloaty_1"(ComponentPose=@n27, Alpha=Get L inward+forward() -> (L inward+forward:double), Translation="0,-1,-3", Rotation="0,0,-8", Scale="1,1,1.21") {
    rotationMode=BMM_Additive
    rotationSpace=BCS_BoneSpace
    scaleMode=BMM_Additive
    scaleSpace=BCS_BoneSpace
    translationMode=BMM_Additive
    translationSpace=BCS_BoneSpace
    alphaBoolBlend {blendOption=Linear}
    alphaScaleBiasClamp {bInterpResult=true}
    boneToModify {boneName=dyn_L_hipFloaty_1}
    rotation {roll=-8}
    scale {z=1.21}
    translation {y=-1, z=-3}
   }
   n29 = "Transform (Modify) Bone - Bone: dyn_L_hipFloaty_1"(ComponentPose=@n28, Alpha=Get L_thighForward() -> (L_thighForward:double), Translation="0,0.4,-1.5", Rotation="0,0,-3", Scale="1,1,1") {
    rotationMode=BMM_Additive
    rotationSpace=BCS_BoneSpace
    translationMode=BMM_Additive
    translationSpace=BCS_BoneSpace
    alphaBoolBlend {blendOption=Linear}
    alphaScaleBiasClamp {bInterpResult=true}
    boneToModify {boneName=dyn_L_hipFloaty_1}
    rotation {roll=-3}
    translation {y=0.4, z=-1.5}
   }
   n30 = "Transform (Modify) Bone - Bone: dyn_R_hipFloaty_front"(ComponentPose=@n29, Alpha=@n19, Translation="1.298,-1.727,3.3", Rotation="-10,1,-3.656", Scale="1,1,1") {
    rotationMode=BMM_Additive
    rotationSpace=BCS_BoneSpace
    translationMode=BMM_Additive
    alphaBoolBlend {blendOption=Linear}
    boneToModify {boneName=dyn_R_hipFloaty_front}
    rotation=<rot -10 1 -3.656>
    translation=<1.298 -1.727 3.3>
   }
   n31 = "Transform (Modify) Bone - Bone: dyn_R_hipFloaty_1"(ComponentPose=@n30, Alpha=@n19, Translation="0,0,0", Rotation="0,0,0", Scale="1.05,1,1.1") {
    scaleMode=BMM_Additive
    scaleSpace=BCS_BoneSpace
    alphaBoolBlend {blendOption=Linear}
    boneToModify {boneName=dyn_R_hipFloaty_1}
    scale {x=1.05, z=1.1}
   }
   n32 = "Transform (Modify) Bone - Bone: dyn_R_hipFloaty_1"(ComponentPose=@n31, Alpha=Get R inward+forward() -> (R inward+forward:double), Translation="2,-2,2", Rotation="0,0,0", Scale="1,1,1") {
    translationMode=BMM_Additive
    alphaBoolBlend {blendOption=Linear}
    alphaScaleBiasClamp {bInterpResult=true}
    boneToModify {boneName=dyn_R_hipFloaty_1}
    translation=<2 -2 2>
   }
   Save cached pose 'startPose'(Pose=Component To Local(ComponentPose=@n32)) {cacheName=startPose}
  ]}
 EventGraph {nodes [
   Event BlueprintUpdateAnimation()
   > n1 = Sequence!(execute) -> (then_0, then_1, then_2:Exec, then_3:Exec, then_4:Exec, then_5:Exec, then_6:Exec, then_7:Exec)
   n2 = Branch(execute=@n1.then_7, Condition=Get bIsInFrontEnd())
   n3 = Cast To FortPlayerPawn(execute=@n25.then, Object=GetOwningActor())
   > n4 = Set Pawn(Pawn:Fort Player Pawn Object Reference=@n3) -> (Output_Get:Fort Player Pawn Object Reference)
   > n5 = Set Anim_Inst(Anim_Inst:Anim Instance Object Reference=GetAnimInstance(self=Get Mesh(self=@n4))) -> (Output_Get:Anim Instance Object Reference)
   Is Valid(exec=@n2.then, InputObject=@n5)
   n6 = Get bIsPlayingEmote()
   > n7 = Set customFrontendIdleOn(customFrontendIdleOn:bool=AND Boolean(
     A=Montage_IsPlaying(self=@n5, Montage=/BRCosmetics/Animation/Game/MainPlayer/Menu/BR/CustomSkins/_s23/SunlitF/SunlitArmor_Female_Idle_Montage)
     B=NOT Boolean(A=@n6)
    )) -> (Output_Get:bool)
   n8 = Get CurrentLOD()
   > Set customFrontendIdleOnAtLOD1(customFrontendIdleOnAtLOD1:bool=AND Boolean(A=@n7, B="Equal (Integer)"(A=@n8, B=1))) -> (Output_Get:bool)
   n9 = Gravity Override State Function DEPRECATED(
    execute=@n1.then_3
    CharacterPartAnimInstance=Self-Reference()
    GameTimeInSeconds=GetGameTimeInSeconds()
    GravityOverrideParams=Get Gravity Override Params()
   )
   > Set Gravity Override(Gravity Override=@n9)
   > n10 = Branch(Condition=Get bIsAccelerating2D())
   > Set isAccelerating2D(isAccelerating2D:double=1) -> (Output_Get:double)
   > Set isNotAccelerating2D(isNotAccelerating2D:double=0) -> (Output_Get:double)
   Set isNotAccelerating2D(execute=@n10.else, isNotAccelerating2D=1)
   > Set isAccelerating2D(isAccelerating2D=0)
   Set L inward+forward(execute=@n1.then_2, L inward+forward:double=float + float(A=Get L_thighInward() -> (L_thighInward:double), B=Get L_thighForward())) -> (Output_Get:double)
   n11 = GetOwnerSkeletalMeshComponent()
   n12 = BreakRotator(InRot=BreakTransform(InTransform=GetDeltaTransformFromRefPose(self=@n11, BoneName=thigh_r, BaseName=pelvis)).Rotation)
   > Set R_thighForward(R_thighForward:double=MapRangeClamped(Value=@n12.Yaw, InRangeA=-60, InRangeB=-90)) -> (Output_Get:double)
   > n13 = Branch(Condition=Get bIsCrouching())
   > Set R_thighInward(R_thighInward:double=MapRangeClamped(Value=@n12.Pitch, InRangeA=10, InRangeB=-7)) -> (Output_Get:double)
   Set R_thighInward(execute=@n13.else, R_thighInward=0)
   n14 = Branch(execute=@n1.then_1, Condition=AND Boolean(A=Get bIsCrouching(), B=Get bIsAccelerating2D()))
   n15 = BreakRotator(InRot=BreakTransform(InTransform=GetDeltaTransformFromRefPose(self=@n11, BoneName=thigh_l, BaseName=pelvis)).Rotation)
   > Set L_thighForward(L_thighForward:double=MapRangeClamped(Value=@n15.Yaw, InRangeA=-40, InRangeB=-90)) -> (Output_Get:double)
   Set L_thighForward(execute=@n14.else, L_thighForward=0)
   Set R inward+forward(execute=@n1.then_4, R inward+forward:double=float + float(A=Get R_thighForward(), B=Get R_thighInward() -> (R_thighInward:double))) -> (Output_Get:double)
   n16 = BreakRotator(InRot=BreakTransform(InTransform=GetDeltaTransformFromRefPose(self=@n11, BoneName=neck_02, BaseName=spine_04)).Rotation)
   Set neckTilted R(
    execute=@n1.then_5
    neckTilted R:double=float * float(A=MapRangeClamped(Value=@n16, InRangeA=5, InRangeB=15), B=Get R upperarm forward() -> (R upperarm forward:double)) {is=title K2Node_PromotableOperator}
   ) -> (Output_Get:double)
   n17 = AND Boolean(A=Get bIsAccelerating2D(), B=Get bIsCrouching())
   n18 = BreakRotator(InRot=BreakTransform(InTransform=GetDeltaTransformFromRefPose(self=@n11, BoneName=neck_01, BaseName=spine_05)).Rotation)
   > Set neckTiltedBack(neckTiltedBack:double=float * float(A=MapRangeClamped(Value=@n18.Yaw, InRangeA=0, InRangeB=-30), B="To Float (Boolean)"(InBool=@n17))) -> (Output_Get:double)
   Get!(execute:Exec=@n1.then_6) -> (then:Exec, else:Exec, Anim_Inst:Anim Instance Object Reference)
   n19 = Get Anim_Inst() -> (Anim_Inst:Anim Instance Object Reference)
   > n20 = Branch(Condition="float > float"(A=GetCurveValue(self=@n19, CurveName=Backpack_Anim_Override)))
   > Set BackPackAlpha(BackPackAlpha:double=0) -> (Output_Get:double)
   Set BackPackAlpha(execute=@n20.else, BackPackAlpha=1)
   n21 = BreakTransform(InTransform=GetDeltaTransformFromRefPose(self=@n11, BoneName=spine_05, BaseName=upperarm_l))
   Set L upperarm up(execute=@n1.then_0, L upperarm up:double=MapRangeClamped(Value=BreakRotator(InRot=@n21.Rotation).Yaw, InRangeA=70, InRangeB=110)) -> (Output_Get:double)
   n22 = BreakRotator(InRot=BreakTransform(InTransform=GetDeltaTransformFromRefPose(self=@n11, BoneName=spine_05, BaseName=upperarm_r)).Rotation)
   > Set R upperarm up(R upperarm up:double=MapRangeClamped(Value=@n22.Yaw, InRangeA=70, InRangeB=110)) -> (Output_Get:double)
   > Set R upperarm forward(R upperarm forward:double=MapRangeClamped(Value=@n22.Pitch, InRangeA=5, InRangeB=15)) -> (Output_Get:double)
   > n23 = Branch(Condition=Get bIsCrouching())
   > Set L_thighInward(L_thighInward:double=MapRangeClamped(Value=@n15.Pitch, InRangeA=10, InRangeB=-7)) -> (Output_Get:double)
   Set L_thighInward(execute=@n23.else, L_thighInward=0)
   Event BlueprintInitializeAnimation()
   n24 = GetOwnerSkeletalMeshComponent()
   > Is Valid(InputObject=@n24)
   > n25 = Set Skeletal_Mesh(Skeletal_Mesh:Skeletal Mesh Component Object Reference=@n24) -> (Output_Get:Skeletal Mesh Component Object Reference)
   Event OnBeginJumping()
   > Set jumpFalling(jumpFalling:double=1) -> (Output_Get:double)
   > Delay(Duration=0.75)
   > Set jumpFalling(jumpFalling=0)
  ]}
]
variables [
 Anim_Inst {type=AnimInstance, default=null}
 BackPackAlpha {type=number}
 customFrontendIdleOn {type=boolean}
 customFrontendIdleOnAtLOD1 {type=boolean}
 Gravity Override {type=Vector}
 Gravity Override Params {type=GravityOverrideParamsStruct, default {onGround_Falling {gravityOverride=<0 0 100>}, onGround_Moving {gravityOverride=<0 0 -2000>}}}
 isAccelerating2D {type=number}
 isNotAccelerating2D {type=number}
 jumpFalling {type=number}
 L inward+forward {type=number}
 L upperarm up {type=number}
 L_thighForward {type=number}
 L_thighInward {type=number}
 neckTilted R {type=number}
 neckTiltedBack {type=number}
 Pawn {type=FortPlayerPawn}
 R inward+forward {type=number}
 R upperarm forward {type=number}
 R upperarm up {type=number}
 R_thighForward {type=number}
 R_thighInward {type=number}
 Skeletal_Mesh {type=SkeletalMeshComponent}
]
