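# Physics asset for the Sunlit Armor face accessory on the medium female player body, giving its front and back braids physics bodies.
# Two capsule chains, a six-segment back braid and a four-segment front braid, hang from the head through base bones, with twist locked and narrow swing limits.
# Capsules on spine, neck, head, clavicles and only the right upper arm form body volumes, joined by freely rotating constraints.
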PhysicsAsset /BRCosmetics/Characters/Player/Female/Medium/Bodies/F_MED_Sunlit_Armor/Meshes/Parts/F_MED_Sunlit_Armor_FaceAcc_Physics
bodies {
 clavicle_l [
  {
   shapeType=Capsule
   length=9.6391
   radius=8
   shapeName=clavicle_l_capsule
   center=<1.8624 0.0004 -2.1272>
   rotation=<rot 90 -179.5524 175.4437>
  }
  {shapeType=Capsule, length=30, radius=6, shapeName=clavicle_l_capsule, center=<15 0 13>, rotation=<rot -20 0 4>}
 ]
 clavicle_r [
  {shapeType=Capsule, length=29.2815, radius=25, shapeName=clavicle_l_capsule, center=<-37.2423 0 -10.1754>, rotation=<rot 0 0 90>}
  {shapeType=Capsule, length=30, radius=7, shapeName=clavicle_r_capsule, center=<-15 0 -13>, rotation=<rot -20 0 -4>}
 ]
 dyn_backBraid_1 [{shapeType=Capsule, length=5, radius=1.1, shapeName=dyn_backBraid_1_capsule, center=<2.5 0 0>, rotation=<rot 90 0 0>}]
 dyn_backBraid_2 [{shapeType=Capsule, length=5, radius=1.1, shapeName=dyn_backBraid_2_capsule, center=<2.5 0 0>, rotation=<rot 90 0 0>}]
 dyn_backBraid_3 [{shapeType=Capsule, length=5, radius=1.1, shapeName=dyn_backBraid_3_capsule, center=<2.5 0 0>, rotation=<rot 90 0 0>}]
 dyn_backBraid_4 [{shapeType=Capsule, length=5, radius=1.1, shapeName=dyn_backBraid_4_capsule, center=<2.5 0 0>, rotation=<rot 90 0 0>}]
 dyn_backBraid_5 [{shapeType=Capsule, length=5, radius=1.1, shapeName=dyn_backBraid_5_capsule, center=<2.5 0 0>, rotation=<rot 90 0 0>}]
 dyn_backBraid_6 [{shapeType=Capsule, length=5, radius=1.1, shapeName=dyn_backBraid_6_capsule, center=<2.5 0 0>, rotation=<rot 90 0 0>}]
 dyn_backBraid_baseLookAt [{shapeType=Capsule, length=5, radius=6, shapeName=dyn_backBraid_1_capsule, center=<2.5 0 0>, rotation=<rot 75 90 90>}]
 dyn_frontBraid_1 [{shapeType=Capsule, length=3, radius=1.6, shapeName=dyn_frontBraid_1_capsule, center=<2 0 0>, rotation=<rot 90 0 0>}]
 dyn_frontBraid_2 [{shapeType=Capsule, length=3, radius=1.4, shapeName=dyn_frontBraid_2_capsule, center=<2 0 0>, rotation=<rot 90 0 0>}]
 dyn_frontBraid_3 [{shapeType=Capsule, length=3, radius=1.4, shapeName=dyn_frontBraid_3_capsule, center=<2 0 0>, rotation=<rot 90 0 0>}]
 dyn_frontBraid_4 [{shapeType=Capsule, length=3, radius=1.4, shapeName=dyn_frontBraid_4_capsule, center=<2 0 0>, rotation=<rot 90 0 0>}]
 dyn_frontBraid_base [{shapeType=Capsule, length=3, radius=1.8, shapeName=dyn_frontBraid_1_capsule, center=<3 0 0>, rotation=<rot 90 0 0>}]
 head [{shapeType=Capsule, length=5, radius=3.6, shapeName=head_capsule, rotation=<rot 90 0 0>}]
 neck_01 [{shapeType=Capsule, length=15, radius=5, shapeName=neck_02_capsule, center=<-5.6553 0 0>}]
 neck_02 [
  {shapeType=Capsule, length=2, radius=6.5, shapeName=neck_02_capsule, center=<-6.156 -0.1346 0>}
  {shapeType=Capsule, length=7, radius=2.5, shapeName=neck_02_capsule, center=<-1.65 -0.7 -6>, rotation=<rot 60 0 -40>}
  {shapeType=Capsule, length=5, radius=1, shapeName=neck_02_capsule, center=<-1.65 -0.7 6>, rotation=<rot -51.8464 10.4093 31.3488>}
 ]
 spine_04 [
  {shapeType=Capsule, length=12, radius=9, shapeName=spine_05_capsule, center=<-0.6664 0.2516 0>}
  {shapeType=Capsule, length=10, radius=5, shapeName=spine_05_capsule, center=<5.014 -5.7711 0>, rotation=<rot 90 165.9638 157.771>}
 ]
 spine_05 [
  {shapeType=Capsule, length=12, radius=9, shapeName=spine_05_capsule, center=<-0.6664 0.2516 0>}
  {shapeType=Capsule, length=10, radius=5, shapeName=spine_05_capsule, center=<5.014 -5.7711 0>, rotation=<rot 90 165.9638 157.771>}
 ]
 upperarm_r [{shapeType=Capsule, length=20, radius=3.5, shapeName=upperarm_r_capsule, center=<-10 0 0>, rotation=<rot 90 0 0>}]
}
constraints [
 {
  bone1Name=neck_02
  bone2Name=neck_01
  swing1LimitDegrees=45
  swing1Motion=Free
  swing2LimitDegrees=45
  swing2Motion=Free
  twistLimitDegrees=45
  twistMotion=Free
 }
 {
  bone1Name=head
  bone2Name=neck_02
  swing1LimitDegrees=45
  swing1Motion=Free
  swing2LimitDegrees=45
  swing2Motion=Free
  twistLimitDegrees=45
  twistMotion=Free
 }
 {
  bone1Name=dyn_backBraid_1
  bone2Name=dyn_backBraid_baseLookAt
  swing1LimitDegrees=40
  swing2LimitDegrees=60
  twistLimitDegrees=45
  twistMotion=Locked
 }
 {bone1Name=dyn_backBraid_2, bone2Name=dyn_backBraid_1, swing1LimitDegrees=20, swing2LimitDegrees=20, twistLimitDegrees=45, twistMotion=Locked}
 {bone1Name=dyn_backBraid_3, bone2Name=dyn_backBraid_2, swing1LimitDegrees=20, swing2LimitDegrees=20, twistLimitDegrees=45, twistMotion=Locked}
 {bone1Name=dyn_backBraid_4, bone2Name=dyn_backBraid_3, swing1LimitDegrees=20, swing2LimitDegrees=20, twistLimitDegrees=45, twistMotion=Locked}
 {bone1Name=dyn_backBraid_5, bone2Name=dyn_backBraid_4, swing1LimitDegrees=20, swing2LimitDegrees=20, twistLimitDegrees=45, twistMotion=Locked}
 {bone1Name=dyn_backBraid_6, bone2Name=dyn_backBraid_5, swing1LimitDegrees=20, swing2LimitDegrees=20, twistLimitDegrees=45, twistMotion=Locked}
 {
  bone1Name=dyn_frontBraid_1
  bone2Name=dyn_frontBraid_base
  swing1LimitDegrees=30
  swing2LimitDegrees=30
  twistLimitDegrees=45
  twistMotion=Locked
 }
 {
  bone1Name=dyn_frontBraid_2
  bone2Name=dyn_frontBraid_1
  swing1LimitDegrees=30
  swing2LimitDegrees=30
  twistLimitDegrees=45
  twistMotion=Locked
 }
 {
  bone1Name=dyn_frontBraid_3
  bone2Name=dyn_frontBraid_2
  swing1LimitDegrees=30
  swing2LimitDegrees=30
  twistLimitDegrees=45
  twistMotion=Locked
 }
 {
  bone1Name=dyn_frontBraid_4
  bone2Name=dyn_frontBraid_3
  swing1LimitDegrees=30
  swing2LimitDegrees=30
  twistLimitDegrees=45
  twistMotion=Locked
 }
 {
  bone1Name=neck_01
  bone2Name=spine_05
  swing1LimitDegrees=45
  swing1Motion=Free
  swing2LimitDegrees=45
  swing2Motion=Free
  twistLimitDegrees=45
  twistMotion=Free
 }
 {
  bone1Name=spine_05
  bone2Name=spine_04
  swing1LimitDegrees=45
  swing1Motion=Free
  swing2LimitDegrees=45
  swing2Motion=Free
  twistLimitDegrees=45
  twistMotion=Free
 }
 {
  bone1Name=clavicle_l
  bone2Name=spine_05
  swing1LimitDegrees=45
  swing1Motion=Free
  swing2LimitDegrees=45
  swing2Motion=Free
  twistLimitDegrees=45
  twistMotion=Free
 }
 {bone1Name=dyn_frontBraid_base, bone2Name=head, swing1LimitDegrees=30, swing2LimitDegrees=30, twistLimitDegrees=45, twistMotion=Locked}
 {
  bone1Name=clavicle_r
  bone2Name=spine_05
  swing1LimitDegrees=45
  swing1Motion=Free
  swing2LimitDegrees=45
  swing2Motion=Free
  twistLimitDegrees=45
  twistMotion=Free
 }
 {
  bone1Name=upperarm_r
  bone2Name=clavicle_r
  swing1LimitDegrees=45
  swing1Motion=Free
  swing2LimitDegrees=45
  swing2Motion=Free
  twistLimitDegrees=45
  twistMotion=Free
 }
 {bone1Name=dyn_backBraid_baseLookAt, bone2Name=head, swing1LimitDegrees=20, swing2LimitDegrees=10, twistLimitDegrees=45, twistMotion=Locked}
]
