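# Skydive physics asset for the Sunlit Armor female character's face accessory, simulating a front braid and a back braid.
# Each braid is a chain of small capsules hanging from the head, with swing-limited, twist-locked constraints and the tightest limit at the front braid base.
# Capsules on the upper spine, neck, head, both clavicles and right upper arm form the body around the braids, linked by fully free angular constraints.
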PhysicsAsset /BRCosmetics/Characters/Player/Female/Medium/Bodies/F_MED_Sunlit_Armor/Meshes/Parts/F_MED_Sunlit_Armor_FaceAcc_Physics_Skydive
bodies {
 clavicle_l [
  {
   shapeType=Capsule
   length=9.6391
   radius=8
   shapeName=clavicle_l_capsule
   center=<1.8624 0.0004 -1.841>
   rotation=<rot 90 -179.5524 175.4437>
  }
  {
   shapeType=Capsule
   length=29.2815
   radius=5.9602
   shapeName=clavicle_l_capsule
   center=<15.1179 -0.2395 12.4027>
   rotation=<rot -20.8685 -0.4329 4.784>
  }
 ]
 clavicle_r [{
   shapeType=Capsule
   length=29.2815
   radius=5.9602
   shapeName=clavicle_l_capsule
   center=<-15.1179 -0.2395 -12.4027>
   rotation=<rot -20.8685 -0.4329 4.784>
  }]
 dyn_backBraid_1 [{shapeType=Capsule, length=5, radius=1.4, shapeName=dyn_backBraid_1_capsule, center=<1 0 0>, rotation=<rot 90 0 0>}]
 dyn_backBraid_2 [{shapeType=Capsule, length=5, radius=1.4, shapeName=dyn_backBraid_2_capsule, center=<1 0 0>, rotation=<rot 90 0 0>}]
 dyn_backBraid_3 [{shapeType=Capsule, length=5, radius=1.4, shapeName=dyn_backBraid_3_capsule, center=<1 0 0>, rotation=<rot 90 0 0>}]
 dyn_backBraid_4 [{shapeType=Capsule, length=5, radius=1.4, shapeName=dyn_backBraid_4_capsule, center=<1 0 0>, rotation=<rot 90 0 0>}]
 dyn_backBraid_5 [{shapeType=Capsule, length=5, radius=1.4, shapeName=dyn_backBraid_5_capsule, center=<1 0 0>, rotation=<rot 90 0 0>}]
 dyn_backBraid_6 [{shapeType=Capsule, length=5, radius=1.4, shapeName=dyn_backBraid_6_capsule, center=<1 0 0>, rotation=<rot 90 0 0>}]
 dyn_frontBraid_1 [{shapeType=Capsule, length=5, radius=1.7, shapeName=dyn_frontBraid_1_capsule, center=<1 0 0>, rotation=<rot 90 0 0>}]
 dyn_frontBraid_2 [{shapeType=Capsule, length=5, radius=1.7, shapeName=dyn_frontBraid_2_capsule, center=<1 0 0>, rotation=<rot 90 0 0>}]
 dyn_frontBraid_3 [{shapeType=Capsule, length=5, radius=1.7, shapeName=dyn_frontBraid_3_capsule, center=<1 0 0>, rotation=<rot 90 0 0>}]
 dyn_frontBraid_4 [{shapeType=Capsule, length=5, radius=1.7, shapeName=dyn_frontBraid_4_capsule, center=<1 0 0>, rotation=<rot 90 0 0>}]
 dyn_frontBraid_base [{shapeType=Capsule, length=5, radius=2, shapeName=dyn_frontBraid_1_capsule, center=<3 0 0>, rotation=<rot 90 0 0>}]
 head [{shapeType=Capsule, length=0.5, radius=0.5, shapeName=head_capsule, rotation=<rot 90 0 0>}]
 neck_01 [{shapeType=Capsule, length=15, radius=5, shapeName=neck_02_capsule, center=<-5.6553 0 0>}]
 neck_02 [
  {shapeType=Capsule, length=6, radius=6, shapeName=neck_02_capsule, center=<-6.156 -1.176 0>}
  {shapeType=Capsule, length=5, radius=1.5, shapeName=neck_02_capsule, center=<-1.8781 -0.4389 -5.7937>, rotation=<rot 60 -0.0004 -25>}
 ]
 spine_05 [{shapeType=Capsule, length=10, radius=5, shapeName=spine_05_capsule, center=<4.8715 -4.7818 0>, rotation=<rot 90 180 -180>}]
 upperarm_r [{shapeType=Capsule, length=29, radius=7, shapeName=upperarm_l_capsule, center=<-11 0 0>, rotation=<rot 90 0 0>}]
}
constraints [
 {
  bone1Name=neck_02
  bone2Name=neck_01
  swing1LimitDegrees=45
  swing1Motion=Free
  swing2LimitDegrees=45
  swing2Motion=Free
  twistLimitDegrees=45
  twistMotion=Free
 }
 {
  bone1Name=head
  bone2Name=neck_02
  swing1LimitDegrees=45
  swing1Motion=Free
  swing2LimitDegrees=45
  swing2Motion=Free
  twistLimitDegrees=45
  twistMotion=Free
 }
 {bone1Name=dyn_backBraid_1, bone2Name=head, swing1LimitDegrees=60, swing2LimitDegrees=60, twistLimitDegrees=45, twistMotion=Locked}
 {bone1Name=dyn_backBraid_2, bone2Name=dyn_backBraid_1, swing1LimitDegrees=30, swing2LimitDegrees=30, twistLimitDegrees=45, twistMotion=Locked}
 {bone1Name=dyn_backBraid_3, bone2Name=dyn_backBraid_2, swing1LimitDegrees=30, swing2LimitDegrees=30, twistLimitDegrees=45, twistMotion=Locked}
 {bone1Name=dyn_backBraid_4, bone2Name=dyn_backBraid_3, swing1LimitDegrees=30, swing2LimitDegrees=30, twistLimitDegrees=45, twistMotion=Locked}
 {bone1Name=dyn_backBraid_5, bone2Name=dyn_backBraid_4, swing1LimitDegrees=30, swing2LimitDegrees=30, twistLimitDegrees=45, twistMotion=Locked}
 {bone1Name=dyn_backBraid_6, bone2Name=dyn_backBraid_5, swing1LimitDegrees=30, swing2LimitDegrees=30, twistLimitDegrees=45, twistMotion=Locked}
 {
  bone1Name=dyn_frontBraid_1
  bone2Name=dyn_frontBraid_base
  swing1LimitDegrees=30
  swing2LimitDegrees=30
  twistLimitDegrees=45
  twistMotion=Locked
 }
 {
  bone1Name=dyn_frontBraid_2
  bone2Name=dyn_frontBraid_1
  swing1LimitDegrees=30
  swing2LimitDegrees=30
  twistLimitDegrees=45
  twistMotion=Locked
 }
 {
  bone1Name=dyn_frontBraid_3
  bone2Name=dyn_frontBraid_2
  swing1LimitDegrees=30
  swing2LimitDegrees=30
  twistLimitDegrees=45
  twistMotion=Locked
 }
 {
  bone1Name=dyn_frontBraid_4
  bone2Name=dyn_frontBraid_3
  swing1LimitDegrees=30
  swing2LimitDegrees=30
  twistLimitDegrees=45
  twistMotion=Locked
 }
 {
  bone1Name=neck_01
  bone2Name=spine_05
  swing1LimitDegrees=45
  swing1Motion=Free
  swing2LimitDegrees=45
  swing2Motion=Free
  twistLimitDegrees=45
  twistMotion=Free
 }
 {
  bone1Name=clavicle_l
  bone2Name=spine_05
  swing1LimitDegrees=45
  swing1Motion=Free
  swing2LimitDegrees=45
  swing2Motion=Free
  twistLimitDegrees=45
  twistMotion=Free
 }
 {bone1Name=dyn_frontBraid_base, bone2Name=head, swing1LimitDegrees=20, swing2LimitDegrees=20, twistLimitDegrees=45, twistMotion=Locked}
 {
  bone1Name=upperarm_r
  bone2Name=clavicle_r
  swing1LimitDegrees=45
  swing1Motion=Free
  swing2LimitDegrees=45
  swing2Motion=Free
  twistLimitDegrees=45
  twistMotion=Free
 }
 {
  bone1Name=clavicle_r
  bone2Name=spine_05
  swing1LimitDegrees=45
  swing1Motion=Free
  swing2LimitDegrees=45
  swing2Motion=Free
  twistLimitDegrees=45
  twistMotion=Free
 }
]
