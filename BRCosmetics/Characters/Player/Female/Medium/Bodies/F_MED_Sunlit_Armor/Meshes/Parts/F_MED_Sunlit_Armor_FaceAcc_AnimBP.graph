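# Post-process style AnimBP for the Sunlit Armor female skin's FaceAcc part, adding secondary motion to braids and strands.
# Copied body pose runs through AnimDynamics strands and a crown fix, then is cached.
# RigidBody simulation picks a backbling physics asset, or switches default/skydive assets with front-braid look-at and menu-idle offset.
# EventGraph derives gravity, backbling state and neck/arm-driven corrective alphas; the skydive-float back-braid transforms are unconnected.
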AnimBlueprint /BRCosmetics/Characters/Player/Female/Medium/Bodies/F_MED_Sunlit_Armor/Meshes/Parts/F_MED_Sunlit_Armor_FaceAcc_AnimBP
target_skeleton=&/F_MED_Sunlit_Armor_FaceAcc_Skeleton
graphs [
 AnimGraph {nodes [
   n1 = Get isSkydiveFloating() -> (isSkydiveFloating:float)
   n2 = "Transform (Modify) Bone - Bone: dyn_backBraid_baseLookAt"!(ComponentPose, Alpha=@n1, Translation="0,0,0", Scale="1,1,1") -> (Pose) {
    rotationMode=BMM_Additive
    alphaBoolBlend {blendOption=Linear}
    boneToModify {boneName=dyn_backBraid_baseLookAt}
    rotation {pitch=0, roll=45}
   }
   "Transform (Modify) Bone - Bone: dyn_backBraid_1"!(ComponentPose=@n2, Alpha=@n1, Translation="0.5,-0.8,-0.5", Scale="1,1,1") -> (Pose) {
    rotationMode=BMM_Additive
    translationMode=BMM_Additive
    alphaBoolBlend {blendOption=Linear}
    boneToModify {boneName=dyn_backBraid_1}
    rotation {roll=-20}
    translation=<0.5 -0.8 -0.5>
   }
   n3 = RigidBody!(
    ComponentPose=Local To Component(LocalPose=Use cached pose 'startPose'())
    bAlphaBoolEnabled:bool
    OverrideWorldGravity:Vector="0,0,-3000"
    ExternalForce
   ) -> (Pose) {
    alphaInputType=Bool
    bClampLinearTranslationLimitToRefPose=true
    bOverrideWorldGravity=true
    overridePhysicsAsset=&/F_MED_Sunlit_Armor_FaceAcc_Physics_Backbling
    simulationSpace=BaseBoneSpace
    alphaBoolBlend {blendOption=Linear}
    baseBoneRef {boneName=spine_02}
    componentLinearAccScale=<0.2 0.2 0.2>
    componentLinearVelScale=<0.25 0.25 2.5>
    overrideWorldGravity {z=-500}
   }
   n4 = Component To Local(ComponentPose=@n3)
   n5 = Look At - Bone: dyn_frontBraid_baseLookAt(ComponentPose=Local To Component(LocalPose=Use cached pose 'startPose'()), Alpha=1) {
    alphaBoolBlend {blendOption=Linear}
    boneToModify {boneName=dyn_frontBraid_baseLookAt}
    lookAtLocation {x=0}
    lookAtTarget {boneReference {boneName=dyn_frontBraid_baseTARGET}}
    lookAt_Axis {axis=<1 0>}
   }
   n6 = Get gravityOverride() -> (gravityOverride:Vector)
   n7 = RigidBody(ComponentPose=@n5, bAlphaBoolEnabled=NOT Boolean(A=Get bIsSkyDiving()), OverrideWorldGravity=@n6) {
    alphaInputType=Bool
    bClampLinearTranslationLimitToRefPose=true
    bOverrideWorldGravity=true
    overridePhysicsAsset=&/F_MED_Sunlit_Armor_FaceAcc_Physics
    simulationSpace=BaseBoneSpace
    alphaBoolBlend {blendOption=Linear}
    baseBoneRef {boneName=spine_02}
    componentLinearAccScale=<0.2 0.2 0.2>
    componentLinearVelScale=<0.25 0.25 2.5>
    overrideWorldGravity {z=-500}
   }
   n8 = RigidBody(ComponentPose=@n7, bAlphaBoolEnabled=Get bIsSkyDiving(), OverrideWorldGravity=@n6) {
    alphaInputType=Bool
    bClampLinearTranslationLimitToRefPose=true
    bOverrideWorldGravity=true
    overridePhysicsAsset=&/F_MED_Sunlit_Armor_FaceAcc_Physics_Skydive
    simulationSpace=BaseBoneSpace
    alphaBoolBlend {blendOption=Linear}
    baseBoneRef {boneName=spine_02}
    componentLinearAccScale=<0.2 0.2 0.2>
    componentLinearVelScale=<0.25 0.25 2.5>
    overrideWorldGravity {z=-500}
   }
   n9 = "Transform (Modify) Bone - Bone: dyn_frontBraid_base"!(
    ComponentPose=@n8
    bAlphaBoolEnabled:bool=Get customFrontendIdleOn() -> (customFrontendIdleOn:bool)
    Translation="-0.5,2.1,0"
    Rotation="0,6,0"
    Scale="1,1,1"
   ) -> (Pose) {
    alphaInputType=Bool
    rotationMode=BMM_Additive
    rotationSpace=BCS_BoneSpace
    translationMode=BMM_Additive
    translationSpace=BCS_BoneSpace
    alphaBoolBlend {blendOption=Linear}
    boneToModify {boneName=dyn_frontBraid_base}
    rotation {yaw=6}
    translation=<-0.5 2.1>
   }
   n10 = Component To Local(ComponentPose=@n9)
   n11 = Get Is Wearing Backbling()
   n12 = Get neckTiltedBack() -> (neckTiltedBack:double)
   n13 = "Transform (Modify) Bone - Bone: dyn_backBraid_baseLookAt"!(
    ComponentPose=Local To Component(LocalPose=Blend Poses by bool(BlendPose_0=@n4, BlendPose_1=@n10, bActiveValue=@n11) {blendType=Linear})
    Alpha=@n12
    Translation="-0.9,1,3.7"
    Rotation:Rotator="-6,0,0"
    Scale="1,1,1"
   ) -> (Pose) {
    rotationMode=BMM_Additive
    rotationSpace=BCS_BoneSpace
    translationMode=BMM_Additive
    translationSpace=BCS_BoneSpace
    alphaBoolBlend {blendOption=Linear}
    alphaScaleBiasClamp {bInterpResult=true}
    boneToModify {boneName=dyn_backBraid_baseLookAt}
    rotation {pitch=-6}
    translation=<-0.9 1 3.7>
   }
   n14 = "Transform (Modify) Bone - Bone: dyn_back_strand"(ComponentPose=@n13, Alpha=@n12, Translation="0,0,0", Rotation="0,0,0", Scale="0.5,1,1") {
    scaleMode=BMM_Additive
    alphaBoolBlend {blendOption=Linear}
    alphaScaleBiasClamp {bInterpResult=true}
    boneToModify {boneName=dyn_back_strand}
    scale {x=0.5}
   }
   Output Pose(Result=Component To Local(ComponentPose=@n14)) {layerGroup=None}
   n15 = Layered blend per bone(
    BasePose=Copy Pose From Mesh() {bCopyCurves=true, bUseAttachedParent=true}
    BlendPoses_0=Copy Pose From Mesh(SourceMeshComponent=Get Body_Mesh())
   ) {layerSetup [{branchFilters [{boneName=upperarm_l}, {boneName=upperarm_r}, {boneName=clavicle_l}, {boneName=clavicle_r}, {boneName=spine_04}, {boneName=spine_05}]}]}
   n16 = Anim Dynamics - Bone: dyn_L_strand_A(ComponentPose=Local To Component(LocalPose=@n15), Alpha=1, ExternalForce="0,0,0") {
    angularDampingOverride=1
    angularSpringConstant=20
    bAngularSpring=true
    bOverrideAngularDamping=true
    bUseGravityOverride=true
    gravityScale=0.35
    simulationSpace=RootRelative
    alphaBoolBlend {blendOption=Linear}
    boundBone {boneName=dyn_L_strand_A}
    componentLinearAccScale=<1 1 1>
    componentLinearVelScale=<1 1 1>
    gravityOverride {z=-2500}
    physicsBodyDefinitions [{
      boundBone {boneName=dyn_L_strand_A}
      boxExtents=<6 6 6>
      constraintSetup {angularLimitsMax=<0 45 0>, angularLimitsMin=<0 -20 -60>, angularTarget=<0 0 0>}
      localJointOffset=<0.5 0 0>
     }]
   }
   n17 = Anim Dynamics - Bone: dyn_back_strand(ComponentPose=@n16, Alpha=1, ExternalForce="0,0,0") {
    angularDampingOverride=1
    angularSpringConstant=20
    bAngularSpring=true
    bOverrideAngularDamping=true
    bUseGravityOverride=true
    gravityScale=0.35
    simulationSpace=RootRelative
    alphaBoolBlend {blendOption=Linear}
    boundBone {boneName=dyn_back_strand}
    componentLinearAccScale=<1 1 1>
    componentLinearVelScale=<1 1 1>
    gravityOverride {z=-2500}
    physicsBodyDefinitions [{
      boundBone {boneName=dyn_back_strand}
      boxExtents=<10 10 10>
      constraintSetup {angularLimitsMax=<0 20 0>, angularLimitsMin=<0 -20 -45>, angularTarget=<0 0 0>}
      localJointOffset=<0.5 0 0>
     }]
   }
   n18 = Anim Dynamics - Bone: dyn_R_strand_A(ComponentPose=@n17, Alpha=1, ExternalForce="0,0,0") {
    angularDampingOverride=1
    angularSpringConstant=20
    bAngularSpring=true
    bOverrideAngularDamping=true
    bUseGravityOverride=true
    gravityScale=0.35
    simulationSpace=RootRelative
    alphaBoolBlend {blendOption=Linear}
    boundBone {boneName=dyn_R_strand_A}
    componentLinearAccScale=<1 1 1>
    componentLinearVelScale=<1 1 1>
    gravityOverride {z=-2500}
    physicsBodyDefinitions [{
      boundBone {boneName=dyn_R_strand_A}
      boxExtents=<6 6 6>
      constraintSetup {angularLimitsMax=<0 10 0>, angularLimitsMin=<0 -45 -60>, angularTarget=<0 0 0>}
      localJointOffset=<0.5 0 0>
     }]
   }
   n19 = Anim Dynamics - Bone: dyn_R_strand_B(ComponentPose=@n18, Alpha=1, ExternalForce="0,0,0") {
    angularDampingOverride=1
    angularSpringConstant=20
    bAngularSpring=true
    bOverrideAngularDamping=true
    bUseGravityOverride=true
    gravityScale=0.35
    simulationSpace=RootRelative
    alphaBoolBlend {blendOption=Linear}
    boundBone {boneName=dyn_R_strand_B}
    componentLinearAccScale=<1 1 1>
    componentLinearVelScale=<1 1 1>
    gravityOverride {z=-2500}
    physicsBodyDefinitions [{
      boundBone {boneName=dyn_R_strand_B}
      boxExtents=<9 9 9>
      constraintSetup {angularLimitsMax=<0 10 0>, angularLimitsMin=<0 -45 -60>, angularTarget=<0 0 0>}
      localJointOffset=<0.5 0 0>
     }]
   }
   n20 = Anim Dynamics - Bone: dyn_R_strand_C(ComponentPose=@n19, Alpha=1, ExternalForce="0,0,0") {
    angularDampingOverride=1
    angularSpringConstant=20
    bAngularSpring=true
    bOverrideAngularDamping=true
    bUseGravityOverride=true
    gravityScale=0.35
    simulationSpace=RootRelative
    alphaBoolBlend {blendOption=Linear}
    boundBone {boneName=dyn_R_strand_C}
    componentLinearAccScale=<1 1 1>
    componentLinearVelScale=<1 1 1>
    gravityOverride {z=-2500}
    physicsBodyDefinitions [{
      boundBone {boneName=dyn_R_strand_C}
      boxExtents=<10 10 10>
      constraintSetup {angularLimitsMax=<0 10 0>, angularLimitsMin=<0 -45 -60>, angularTarget=<0 0 0>}
      localJointOffset=<0.5 0 0>
     }]
   }
   n21 = "Transform (Modify) Bone - Bone: dyn_R_crownFixer"(ComponentPose=@n20, Alpha=Get neckTilted R() -> (neckTilted R:double), Translation="0,-0.1,-0.3", Rotation="0,0,0", Scale="1,1,0.7") {
    scaleMode=BMM_Additive
    scaleSpace=BCS_BoneSpace
    translationMode=BMM_Additive
    translationSpace=BCS_BoneSpace
    alphaBoolBlend {blendOption=Linear}
    boneToModify {boneName=dyn_R_crownFixer}
    scale {z=0.7}
    translation {y=-0.1, z=-0.3}
   }
   Save cached pose 'startPose'(Pose=Component To Local(ComponentPose=@n21)) {cacheName=startPose}
  ]}
 EventGraph {nodes [
   Event BlueprintUpdateAnimation()
   > n1 = Sequence!(execute) -> (then_0, then_1, then_2:Exec)
   > n2 = Gravity Override State Function DEPRECATED(CharacterPartAnimInstance=Self-Reference(), GameTimeInSeconds=GetGameTimeInSeconds(), GravityOverrideParams=Get Gravity Override Params())
   > Set gravityOverride(gravityOverride:Vector=@n2) -> (Output_Get:Vector)
   Set Is Wearing Backbling(execute=@n1.then_1, Is Wearing Backbling=CheckForBackbling())
   > Set isSkydiveFloating(isSkydiveFloating:float="To Float (Boolean)"(InBool=Get bIsSkydiveFloating())) -> (Output_Get:float)
   n3 = GetOwnerSkeletalMeshComponent()
   n4 = BreakTransform(InTransform=GetDeltaTransformFromRefPose(self=@n3, BoneName=upperarm_r, BaseName=clavicle_r))
   > Set R upperarm Forward(R upperarm Forward:double=MapRangeClamped(Value=BreakRotator(InRot=@n4.Rotation).Yaw, InRangeA=20, InRangeB=40)) -> (Output_Get:double)
   n5 = BreakRotator(InRot=BreakTransform(InTransform=GetDeltaTransformFromRefPose(self=@n3, BoneName=neck_01, BaseName=spine_05)).Rotation)
   > Set neckTilted R(neckTilted R:double=float * float(A=MapRangeClamped(Value=@n5.Pitch, InRangeA=-3, InRangeB=7), B=Get R upperarm Forward() -> (R upperarm Forward:double)) {is=title K2Node_PromotableOperator}) -> (Output_Get:double)
   n6 = AND Boolean(A=Get bIsAccelerating2D(), B=Get bIsCrouching())
   n7 = BreakRotator(InRot=BreakTransform(InTransform=GetDeltaTransformFromRefPose(self=@n3, BoneName=neck_01, BaseName=spine_05)).Rotation)
   > Set neckTiltedBack(neckTiltedBack:double=float * float(A=MapRangeClamped(Value=@n7.Yaw, InRangeA=0, InRangeB=-30), B="To Float (Boolean)"(InBool=@n6))) -> (Output_Get:double)
   n8 = Branch(execute=@n1.then_2, Condition=Get bIsInFrontEnd())
   n9 = Cast To FortPlayerPawn(execute=@n14.then, Object=GetOwningActor())
   > n10 = Set Pawn(Pawn:Fort Player Pawn Object Reference=@n9) -> (Output_Get:Fort Player Pawn Object Reference)
   > n11 = Set Anim_Inst(Anim_Inst:Anim Instance Object Reference=GetAnimInstance(self=Get Mesh(self=@n10))) -> (Output_Get:Anim Instance Object Reference)
   Is Valid(exec=@n8.then, InputObject=@n11)
   n12 = Get bIsPlayingEmote()
   > Set customFrontendIdleOn(customFrontendIdleOn:bool=AND Boolean(
     A=Montage_IsPlaying(self=@n11, Montage=/BRCosmetics/Animation/Game/MainPlayer/Menu/BR/CustomSkins/_s23/SunlitF/SunlitArmor_Female_Idle_Montage)
     B=NOT Boolean(A=@n12)
    )) -> (Output_Get:bool)
   Event BlueprintInitializeAnimation()
   n13 = GetOwnerSkeletalMeshComponent()
   > Is Valid(InputObject=@n13)
   > n14 = Set Body_Mesh(Body_Mesh=@n13)
   Set R upperarm up(R upperarm up:double=MapRangeClamped(Value=BreakRotator(InRot="0, 0, 0"), InRangeA=-10, InRangeB=-40)) -> (Output_Get:double)
  ]}
]
variables [
 Anim_Inst {type=AnimInstance, default=null}
 Body_Mesh {type=SkeletalMeshComponent}
 customFrontendIdleOn {type=boolean}
 Gravity Override Params {
  type=GravityOverrideParamsStruct
  default {
   onGround_Moving {gravityOverride=<0 0 -500>}
   onGround_Standing {gravityOverride=<0 0 -500>}
   skydive_FloatingDown {gravityOverride=<0 -700 50>}
  }
 }
 gravityOverride {type=Vector}
 Is Wearing Backbling {type=boolean}
 isSkydiveFloating {type=number}
 L upperarm forward {type=number}
 neckTilted R {type=number}
 neckTiltedBack {type=number}
 Pawn {type=FortPlayerPawn}
 R upperarm Forward {type=number}
 R upperarm up {type=number}
]
